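# Physics asset for the OctopusV2 vehicle in BallerV2, defining its simple collision body.
# It holds a single body on the root bone: a zero-length capsule, effectively a sphere, raised above the bone origin.
# There are no constraints, so the asset describes one rigid shape with no articulated parts.
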PhysicsAsset /BallerV2/Vehicle/OctopusV2VehiclePhysics
bodies {root [{shapeType=Capsule, length=0, radius=75, center=<0.2733 0 76.4117>}]}
constraints []
