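; Blueprint for the V2 Baller (Octopus) vehicle, a drivable ball with boost and a grappling towhook, layering audio, effects and camera feedback onto the native vehicle.
; Vehicle events drive impacts, boost and coaster FOV timelines, water effects, pickaxe knockback, health damage states and destruction.
; The towhook rope is a scaled mesh with material-driven wiggle, updated by a self-rearming timer and retracted by timeline.
; Coaster custom events manage dust and speed-triggered boost sounds with cooldown.
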(event Custom|CameraShake
  (params (Scale :type Float%20%28double-precision%29))
  (Game|Feedback|ClientStartCameraShake :self (Fortnite|GetFortPlayerControllerfromActor (Vehicle|GetDriver)) :Shake (Variables|Default|GetBounceCameraShake) :Scale (Math|Float|MapRangeClamped Scale (Variables|Default|GetBounceEffectsThreshold) (Variables|Default|GetInRangeMax) (Variables|Default|GetInMax) (Variables|Default|GetOutMax)) :PlaySpace "CameraLocal"))

(event Custom|Reset%20Coaster%20Boost%20SFX
  ;[1 0 0.7939] Coaster Boost SFX
  (Utilities|FlowControl|DoOnce false
    (:Completed
      (Variables|Sounds|SetSpeed%20Coaster%20Audio%20Target (Vehicle|GetTotalSpeedKmh))
      (bind _coasterboostaudio (Variables|Default|GetCoasterBoostAudio)
        (:then
          (Audio|Components|Audio|FadeOut _coasterboostaudio 0.5 0.0 "Linear" :!class /Script/Engine.AudioComponent))
        (:else
          _))
      (bind _returnvalue (Audio|SpawnSoundAttached "/BallerV2/Sound/DaisyCoaster_Hamster_Boost_Cue.DaisyCoaster_Hamster_Boost_Cue" (Variables|Default|GetAudioMovement)))
      (Variables|Default|SetCoasterBoostAudio _returnvalue)
      (Variables|Sounds|SetPrevious%20Audio%20Target (Variables|Sounds|GetSpeed%20Coaster%20Audio%20Target))
      (Variables|Sounds|SetCan%20Spawn%20Coaster%20Boost%20Noise false)
      (Utilities|FlowControl|Delay (Variables|Sounds|GetCoaster%20Boost%20Cooldown))
      (Variables|Sounds|SetCan%20Spawn%20Coaster%20Boost%20Noise true))))

(event Custom|Coaster%20Boost%20SFX
  (bind _returnvalue (Vehicle|GetTotalSpeedKmh))
  (bind _speed_coaster_audio_target (Variables|Sounds|GetSpeed%20Coaster%20Audio%20Target))
  ;[1 0 0.7939] Coaster Boost SFX
  (if (Variables|Sounds|GetCan%20Spawn%20Coaster%20Boost%20Noise)
    (Utilities|FlowControl|Sequence
      (:then_0
        (Utilities|FlowControl|DoOnce false
          (:Completed
            (Variables|Sounds|SetSpeed%20Coaster%20Audio%20Target _returnvalue))))
      (:then_1
        (if (Coaster|ShouldUpdateSpeedTarget (Variables|Sounds|GetSpeed%20Coaster%20Audio%20Target))
          (Variables|Sounds|SetSpeed%20Coaster%20Audio%20Target _returnvalue)
          (if (not (> _speed_coaster_audio_target (+ (Variables|Sounds|GetPrevious%20Audio%20Target) (Variables|Coaster|GetMin%20Speed%20Increase))))
            (Variables|Sounds|SetPrevious%20Audio%20Target _speed_coaster_audio_target)))))
    (bind _coasterboostaudio (Variables|Default|GetCoasterBoostAudio)
      (:then
        (Audio|Components|Audio|FadeOut _coasterboostaudio 0.5 0.0 "Linear" :!class /Script/Engine.AudioComponent))
      (:else
        _))
    (bind _returnvalue_2 (Audio|SpawnSoundAttached "/BallerV2/Sound/DaisyCoaster_Hamster_Boost_Cue.DaisyCoaster_Hamster_Boost_Cue" (Variables|Default|GetAudioMovement)))
    (Variables|Default|SetCoasterBoostAudio _returnvalue_2)
    (Variables|Sounds|SetPrevious%20Audio%20Target (Variables|Sounds|GetSpeed%20Coaster%20Audio%20Target))
    (Variables|Sounds|SetCan%20Spawn%20Coaster%20Boost%20Noise false)
    (Utilities|FlowControl|Delay (Variables|Sounds|GetCoaster%20Boost%20Cooldown))
    (Variables|Sounds|SetCan%20Spawn%20Coaster%20Boost%20Noise true)))

(event Custom|GroundHit
  (params (Scale :type Float%20%28double-precision%29))
  (Game|Feedback|ClientStartCameraShake :self (Fortnite|GetFortPlayerControllerfromActor (Vehicle|GetDriver)) :Shake "/Game/Athena/DrivableVehicles/Octopus/OctopusHitGroundShake.OctopusHitGroundShake_C" :Scale Scale :PlaySpace "CameraLocal"))

(event EventBeginPlay
  ;[1 1 1] BeginPlay
  (|CallParentFunction)
  (CallFunction|RefreshDriveParameters) ; Update Baller Speed to match GameData Variables. Execute on both Client and Server to verify values match with each other.
  (Networking|CanExecuteCosmeticEvents
    (:True
      (EventDispatchers|BindEventtoOnTowhookAttached (EventDispatchers|CreateEvent :!event Custom|TowhookAttached))
      (EventDispatchers|BindEventtoOnTowhookDetached (EventDispatchers|CreateEvent :!event Custom|TowhookDetached))
      (EventDispatchers|BindEventtoOnTowhookShot (EventDispatchers|CreateEvent :!event Custom|TowhookShot))
      (EventDispatchers|BindEventtoOnTowhookMissed (EventDispatchers|CreateEvent :!event Custom|TowhookMissed))
      (CallFunction|CacheAudioPointers (Variables|Default|GetAudioMovement) (Variables|Default|GetAudioWind1P) (Variables|Default|GetAudioTowCable))
      (CallFunction|CacheParticleFX (Variables|Default|GetCoilDustVisuals) 0 0 0 0 (Variables|Default|GetCoilDustVisuals))
      (EventDispatchers|BindEventtoOnTowhookAttached (EventDispatchers|CreateEvent :!event Custom|EventOnTowhookAttached))
      (Health|BindEventtoOnHealthChanged (EventDispatchers|CreateEvent :!event Handle%20On%20Health%20Changed)))))

(event Custom|StopCameraShake
  (bind _drivingplayercontroller (Variables|Default|GetDrivingPlayerController))
  ;[1 1 1] Stop Camera Shake
  (Utilities|IsValid (Variables|FortOctopusVehicle|GetCacheDriverCameraShake)
    (:Is%20Valid
      (Utilities|IsValid _drivingplayercontroller
        (:Is%20Valid
          (CameraShakes|StopCameraShake (Variables|/Script/Engine.PlayerController|GetPlayerCameraManager _drivingplayercontroller) (Variables|FortOctopusVehicle|GetCacheDriverCameraShake) false)
          (Variables|FortOctopusVehicle|SetCacheDriverCameraShake 0))))))

(event Custom|Turn%20Off%20Dust
  ;[0.9583 0.1941 1] Turns off Dust Particle FX while on Coaster.
  (if (not (Variables|FX|GetCoil%20FX%20Locked))
    (Utilities|FlowControl|DoOnce false
      (:Completed
        (Effects|Components|ParticleSystem|SetBoolParameter (Variables|Default|GetCoilDustVisuals) "SpawnDust" false)))))

(event Custom|Coaster%20FOV
  (bind _returnvalue (Vehicle|GetTotalSpeedKmh))
  ;[0.2656 0.2155 0] Coaster Speed FOV
  (if (!= (ScalableFloat|GetValueatLevel (Variables|Hotfix|GetCoaster%20FOV%20Multiplier)) 0)
    (Utilities|FlowControl|DoOnce false
      (:Completed
        (Components|Timeline|SetPlayRate (Variables|OctopusVehicle_V2|GetCoaster%20Lurch%20FOV%20Timeline) (Variables|Default|GetCoaster%20FOV%20Timeline%20Play%20Rate))
        (Variables|Coaster|SetSpeed%20Coaster%20FOV%20Target _returnvalue)
        (bind (_direction _boostcurve) (Timeline|Coaster%20Lurch%20FOV%20Timeline PlayFromStart)
          (:Update
            (Variables|Camera|SetCameraFOVOffset (Math|Float|Lerp 0.0 (* _returnvalue (ScalableFloat|GetValueatLevel (Variables|Hotfix|GetCoaster%20FOV%20Multiplier))) _boostcurve)))
          (:Finished
            (bind _returnvalue_2 (Utilities|Time|SetTimerbyEvent (EventDispatchers|CreateEvent :!event Custom|Update%20Speed%20to%20FOV) 0.2 true))
            (Variables|Default|SetCoaster%20FOV%20Timer _returnvalue_2)))))))

(event Custom|Reset%20Dust
  ;[0.9583 0.1941 1] Turns off Dust Particle FX while on Coaster.
  (if (not (Variables|FX|GetCoil%20FX%20Locked))
    (Effects|Components|ParticleSystem|SetBoolParameter (Variables|Default|GetCoilDustVisuals) "SpawnDust" true)
    (Utilities|FlowControl|DoOnce false
      (:Completed
        (Effects|Components|ParticleSystem|SetBoolParameter (Variables|Default|GetCoilDustVisuals) "SpawnDust" false)))))

(event Custom|FOV%20Reset
  (bind _returnvalue (Vehicle|GetTotalSpeedKmh))
  ;[0.2656 0.2155 0] Coaster Speed FOV
  (Utilities|Time|ClearandInvalidateTimerbyHandle (Variables|Default|GetCoaster%20FOV%20Timer))
  (Variables|Camera|SetCameraFOVOffset 0.0)
  (Utilities|FlowControl|DoOnce false
    (:Completed
      (Components|Timeline|SetPlayRate (Variables|OctopusVehicle_V2|GetCoaster%20Lurch%20FOV%20Timeline) (Variables|Default|GetCoaster%20FOV%20Timeline%20Play%20Rate))
      (Variables|Coaster|SetSpeed%20Coaster%20FOV%20Target _returnvalue)
      (Timeline|Coaster%20Lurch%20FOV%20Timeline PlayFromStart))))

(event Custom|StartDriverCameraShake
  (bind _drivingplayercontroller (Variables|Default|GetDrivingPlayerController))
  ;[1 1 1] Start Camera Shake
  (if (Vehicle|ClientIsDriver)
    (if (not (Variables|Default|GetDebug_NoCameraShake))
      (Utilities|IsValid _drivingplayercontroller
        (:Is%20Valid
          (bind _returnvalue (CameraShakes|StartCameraShake :self (Variables|/Script/Engine.PlayerController|GetPlayerCameraManager _drivingplayercontroller) :ShakeClass "/Game/Athena/DrivableVehicles/Speed_CameraShake.Speed_CameraShake_C" :Scale 1.0 :PlaySpace "CameraLocal"))
          (Variables|FortOctopusVehicle|SetCacheDriverCameraShake (CameraShakes|ConvLegacyCameraShake _returnvalue)))))))

(event Custom|Update%20Speed%20to%20FOV
  ;[0.2656 0.2155 0] Coaster Speed FOV
  (if (Coaster|ShouldUpdateSpeedTarget (Variables|Coaster|GetSpeed%20Coaster%20FOV%20Target))
    (Utilities|FlowControl|DoOnce false
      (:Completed
        (Variables|Coaster|SetSpeed%20Coaster%20FOV%20Target (Vehicle|GetTotalSpeedKmh))
        (Variables|Default|SetPrevious%20FOV (Variables|Camera|GetCameraFOVOffset))
        (bind (_direction _alpha) (Timeline|Coaster%20Speed%20FOV%20Timeline PlayFromStart)
          (:Update
            (Variables|Camera|SetCameraFOVOffset (Math|Float|Lerp (Variables|Default|GetPrevious%20FOV) (* (Variables|Coaster|GetSpeed%20Coaster%20FOV%20Target) (ScalableFloat|GetValueatLevel (Variables|Hotfix|GetCoaster%20FOV%20Multiplier))) _alpha)))
          (:Finished))))))

(event EventEndPlay
  (params (EndPlayReason :type EEndPlayReason%20Enum :default "Destroyed"))
  ;[1 1 1] End Play
  (CallFunction|UpdateTowhookinAirAudio false))

(event Collision|EventActorBeginOverlap
  (params (OtherActor :type Actor%20Object%20Reference)))

(event EventOnHit
  (params (Location :type Vector%20%28by%20ref%29 :default "0, 0, 0") (Normal :type Vector%20%28by%20ref%29 :default "0, 0, 0") (Force :type Float%20%28single-precision%29) (HitActor :type Actor%20Object%20Reference))
  (bind _on_hit_normal (Variables|OnHit|GetOn%20Hit%20Normal))
  (bind _returnvalue (Math|Rotator|MakeRotfromX _on_hit_normal))
  (bind _on_hit_location (Variables|OnHit|GetOn%20Hit%20Location))
  ;[1 1 1] Impacts
  (Variables|OnHit|SetOn%20Hit%20Location Location)
  (Variables|OnHit|SetOn%20Hit%20Normal Normal)
  (Variables|OnHit|SetOn%20Hit%20Force Force)
  (Variables|OnHit|SetOn%20Hit%20Actor HitActor)
  (Utilities|FlowControl|Sequence
    (:then_1
      (if (not (Networking|IsDedicatedServer))
        (Utilities|FlowControl|Sequence
          (:then_0
            (if (>= (Variables|OnHit|GetOn%20Hit%20Force) (Variables|Default|GetBounceThreshold))
              (Utilities|FlowControl|DoOnce false
                (:Completed
                  (Utilities|FlowControl|Sequence
                    (:then_0
                      (if (GameplayTags|HasMatchingGameplayTag self "(TagName=\"Gameplay.Octopus.Boosting\")")
                        (Effects|SpawnEmitteratLocation "/Game/Athena/DrivableVehicles/FX/Octopus/P_Octopus_Impact_Boost.P_Octopus_Impact_Boost" _on_hit_location _returnvalue)
                        (if (Variables|Sounds|GetbCanPlayBoostBounceAudio)
                          (Audio|PlaySoundatLocation "/Game/Athena/Sounds/Vehicles/Hamster/SC_Hamster_CollisionImpact_Boost_Cue.SC_Hamster_CollisionImpact_Boost_Cue" _on_hit_location "0, 0, 0" 1.0 (Math|Float|MapRangeClamped (Math|Conversions|ToFloat(Integer) 0) 0.0 12.0 0.9 2.0)))
                        (else
                          (Effects|SpawnEmitteratLocation "/Game/Athena/DrivableVehicles/FX/Octopus/P_Octopus_Impact.P_Octopus_Impact" _on_hit_location _returnvalue))))
                    (:then_1
                      (Utilities|Time|SetTimerbyEvent (EventDispatchers|CreateEvent :!event Custom|ImpactReset) (Variables|Default|GetImpactResetTimer))
                      (if (Vehicle|ClientIsDriver)
                        (CallFunction|CameraShake (Variables|OnHit|GetOn%20Hit%20Force)))))))))
          (:then_1
            ;; Audio Collision_2
            (if (not (GameplayTags|HasMatchingGameplayTag (Variables|Default|GetRiding%20Coaster%20Tag)))
              (Audio|OnCollision (Variables|Default|GetNormalCollisionAudio) (+ (Variables|OnHit|GetOn%20Hit%20Location) "0,0,50.000000") (* (Variables|OnHit|GetOn%20Hit%20Normal) (Variables|OnHit|GetOn%20Hit%20Force))))))))))

(event EventOnPawnEnterVehicle
  (params (PlayerPawn :type Fort%20Player%20Pawn%20Object%20Reference) (SeatIdx :type Integer))
  (bind _coildustvisuals (Variables|Default|GetCoilDustVisuals))
  (bind _audiogetinout (Variables|Default|GetAudioGetInOut))
  ;[1 1 1] Player In and Out
  (CallFunction|AddCooldownCueEvent PlayerPawn)
  (Utilities|FlowControl|Sequence
    (:then_0
      (bind _asplayer_controller (Utilities|Casting|CastToPlayerController (Pawn|GetController PlayerPawn)))
      (bind _output_get (Variables|Default|SetDrivingPlayerController _asplayer_controller))
      (CallFunction|StartDriverCameraShake)
      (CallFunction|ToggleHUDElements _output_get true))
    (:then_1
      (Components|Activation|Activate (Variables|Default|GetCoilDustVisuals) true)
      (Audio|Components|Audio|SetSound _audiogetinout (Variables|Sounds|GetSoundOnGetIn))
      (Audio|Components|Audio|Play _audiogetinout 0.0)
      (bind (_direction _intensity) (Timeline|Startup PlayFromStart)
        (:Update
          (Effects|Components|ParticleSystem|SetFloatParameter _coildustvisuals "BoostAlpha" _intensity)
          (Effects|Components|ParticleSystem|SetFloatParameter _coildustvisuals "Boost" _intensity))))
    (:then_2
      ;; EquipWeapon
      (if (Vehicle|IsOperatingTurret PlayerPawn)
        (bind _returnvalue (Environment|ApplyGlobalEnvironmentGameplayEffecttoActor PlayerPawn (Variables|Default|GetGE_GrantWeapon)))
        (Variables|Default|SetGE_EnterTurret _returnvalue)))))

(event Custom|ImpactReset
  (bind _on_hit_normal (Variables|OnHit|GetOn%20Hit%20Normal))
  (bind _returnvalue (Math|Rotator|MakeRotfromX _on_hit_normal))
  (bind _on_hit_location (Variables|OnHit|GetOn%20Hit%20Location))
  ;[1 1 1] Impacts
  (Utilities|FlowControl|DoOnce false
    (:Completed
      (Utilities|FlowControl|Sequence
        (:then_0
          (if (GameplayTags|HasMatchingGameplayTag self "(TagName=\"Gameplay.Octopus.Boosting\")")
            (Effects|SpawnEmitteratLocation "/Game/Athena/DrivableVehicles/FX/Octopus/P_Octopus_Impact_Boost.P_Octopus_Impact_Boost" _on_hit_location _returnvalue)
            (if (Variables|Sounds|GetbCanPlayBoostBounceAudio)
              (Audio|PlaySoundatLocation "/Game/Athena/Sounds/Vehicles/Hamster/SC_Hamster_CollisionImpact_Boost_Cue.SC_Hamster_CollisionImpact_Boost_Cue" _on_hit_location "0, 0, 0" 1.0 (Math|Float|MapRangeClamped (Math|Conversions|ToFloat(Integer) 0) 0.0 12.0 0.9 2.0)))
            (else
              (Effects|SpawnEmitteratLocation "/Game/Athena/DrivableVehicles/FX/Octopus/P_Octopus_Impact.P_Octopus_Impact" _on_hit_location _returnvalue))))
        (:then_1
          (Utilities|Time|SetTimerbyEvent (EventDispatchers|CreateEvent :!event Custom|ImpactReset) (Variables|Default|GetImpactResetTimer))
          (if (Vehicle|ClientIsDriver)
            (CallFunction|CameraShake (Variables|OnHit|GetOn%20Hit%20Force))))))))

(event EventOnPawnExitVehicle
  (params (PlayerPawn :type Fort%20Player%20Pawn%20Object%20Reference) (ExitSocketName :type Name :default "None"))
  (bind _coildustvisuals (Variables|Default|GetCoilDustVisuals))
  (bind _audiogetinout (Variables|Default|GetAudioGetInOut))
  ;[1 1 1] Player In and Out
  (Utilities|FlowControl|Sequence
    (:then_0
      (CallFunction|ToggleHUDElements (Variables|Default|GetDrivingPlayerController))
      (Components|Activation|Deactivate (Variables|Default|GetCoilDustVisuals))
      (Audio|Components|Audio|SetSound _audiogetinout (Variables|Sounds|GetSoundOnGetOut))
      (Audio|Components|Audio|Play _audiogetinout 0.0)
      (Timeline|Startup ReverseFromEnd))
    (:then_1
      (CallFunction|StopCameraShake)
      (Development|SetHiddeninGame (Variables|Default|GetReticle) true false))
    (:then_2
      (Utilities|IsValid PlayerPawn
        (:Is%20Valid
          ;; UnequipWeapon
          (GameplayEffects|RemoveActiveGameplayEffect (Ability|GetAbilitySystemComponent PlayerPawn) (Variables|Default|GetGE_EnterTurret) -1))))
    (:then_3
      (Utilities|IsValid PlayerPawn
        (:Is%20Valid
          (CallFunction|RemoveCooldownCueEvent PlayerPawn)
          (CallFunction|DestroyTowhookProjectile))))))

(event EventOnBoostBegin
  (bind _coildustvisuals (Variables|Default|GetCoilDustVisuals))
  ;[1 1 1] Applies Boosting tag
  (Environment|ApplyGlobalEnvironmentGameplayEffecttoActor self "/BallerV2/Vehicle/GE_OctopusVehicle_IsBoosting.GE_OctopusVehicle_IsBoosting_C")
  ;[1 1 1] Boost FX
  (Default|CallDispatchBoostStarted)
  (CallFunction|BoostCameraBegin)
  (CallFunction|BoostVFX true)
  (Effects|Components|ParticleSystem|SetFloatParameter _coildustvisuals "BoostAlpha" 1.0)
  (Effects|Components|ParticleSystem|SetFloatParameter _coildustvisuals "Boost" 1.0)
  (Audio|Components|Audio|Play (Variables|Default|GetAudioBoost) 0.0)
  (Utilities|Time|SetTimerbyEvent (delegate Custom|Bounce%20Collision%20Audio%20Cooldown) 0.2)
  (Variables|Sounds|SetbCanPlayBoostBounceAudio false))

(event Custom|Bounce%20Collision%20Audio%20Cooldown
  ;[1 1 1] Boost FX
  (Variables|Sounds|SetbCanPlayBoostBounceAudio true))

(event EventOnBoostEnd
  (bind _coildustvisuals (Variables|Default|GetCoilDustVisuals))
  ;[1 1 1] Applies Boosting tag
  (GameplayEffects|RemoveActiveEffectswithGrantedTags (Vehicle|GetVehicleAbilitySystemComponent self) "(GameplayTags=((TagName=\"Gameplay.Octopus.Boosting\")))")
  ;[1 1 1] Boost FX
  (Default|CallDispatchBoostEnded)
  (CallFunction|BoostCameraReverse)
  (CallFunction|BoostVFX false)
  (Effects|Components|ParticleSystem|SetFloatParameter _coildustvisuals "BoostAlpha" 0.0)
  (Effects|Components|ParticleSystem|SetFloatParameter _coildustvisuals "Boost" 0.0))

(event Custom|BoostCameraBegin
  ;[1 1 1] Boost Effects
  (Variables|Default|SetbBoostCameraActive true)
  (Components|Timeline|SetPlayRate (Variables|OctopusVehicle_V2|GetBoostTimeline) (Variables|Default|GetBoostTimelinePlayRate))
  (bind (_direction _rumble _newtrack_0) (Timeline|BoostTimeline Play)
    (:Update
      (Variables|Camera|SetCameraFOVOffset (Math|Float|Lerp 0.0 (Variables|Default|GetBoostFOVOffset) _newtrack_0))
      (Variables|Camera|SetCameraSpaceForwardDistanceOffset (Math|Float|Lerp 0.0 (Variables|Default|GetBoostCameraDistanceOffset) _newtrack_0))
      (Variables|Default|SetBoostRumbleValue _rumble)
      (switch Utilities|FlowControl|Switch|SwitchonETimelineDirection _direction
        (:Forward
          (Rendering|Material|SetScalarParameterValueonMaterials (Variables|Vehicle|GetSkeletalMesh) "BoostGlow" (Math|Float|Clamp(Float) (* _newtrack_0 1.0))))))
    (:Finished
      (switch Utilities|FlowControl|Switch|SwitchonETimelineDirection _direction
        (:Forward
          (CallFunction|BoostCameraReverse))
        (:Backward
          (Variables|Default|SetbBoostCameraActive false))))))

(event Custom|ReverseBoostTimeline
  ;[1 1 1] Boost Effects
  (Timeline|BoostTimeline Reverse))

(event Custom|BoostCameraReverse
  ;[1 1 1] Boost Effects
  (if (Variables|Default|GetbBoostCameraActive)
    (Components|Timeline|SetPlayRate (Variables|OctopusVehicle_V2|GetBoostTimeline) (Variables|Default|GetBoostReversePlayRate))
    (CallFunction|ReverseBoostTimeline)))

(event EventOnUpdateTargetPoint
  (params (bAimAssistHitsTarget :type Boolean) (TargetPoint :type Vector%20%28by%20ref%29 :default "0, 0, 0"))
  (bind _reticle (Variables|Default|GetReticle))
  (Development|SetHiddeninGame _reticle (not bAimAssistHitsTarget) false)
  (Transformation|SetWorldLocation _reticle TargetPoint false false))

(event Custom|EventOnTowhookAttached
  (Development|SetHiddeninGame (Variables|Default|GetReticle) true false))

(event EventK2_ApplyCosmeticWrap
  (params (LoadedWrap :type Athena%20Item%20Wrap%20Definition%20Object%20Reference))
  (bind _output_materialinterface (Utilities|Struct|BreakSkeletalMaterial (Utilities|Array|Get(acopy) (Variables|/Script/Engine.SkeletalMesh|GetMaterials (Variables|/Script/Engine.SkinnedMeshComponent|GetSkeletalMesh (Variables|Vehicle|GetSkeletalMesh))))))
  (Class|FortVehicleInterface|ApplyMaterialandUseWrapifPossible _output_materialinterface))

(event EventOnDamagePlayEffects
  (params (Damage :type Float%20%28single-precision%29) (DamageTags :type Gameplay%20Tag%20Container%20Structure%20%28by%20ref%29) (Momentum :type Vector :default "0, 0, 0") (HitInfo_bBlockingHit :type Boolean) (HitInfo_bInitialOverlap :type Boolean) (HitInfo_Time :type Float%20%28single-precision%29) (HitInfo_Distance :type Float%20%28single-precision%29) (HitInfo_Location :type Vector :default "0, 0, 0") (HitInfo_ImpactPoint :type Vector :default "0, 0, 0") (HitInfo_Normal :type Vector :default "0, 0, 0") (HitInfo_ImpactNormal :type Vector :default "0, 0, 0") (HitInfo_PhysMat :type Physical%20Material%20Object%20Reference) (HitInfo_HitActor :type Actor%20Object%20Reference) (HitInfo_HitComponent :type Primitive%20Component%20Object%20Reference) (HitInfo_HitBoneName :type Name :default "None") (HitInfo_BoneName :type Name :default "None") (HitInfo_HitItem :type Integer) (HitInfo_ElementIndex :type Integer) (HitInfo_FaceIndex :type Integer) (HitInfo_TraceStart :type Vector :default "0, 0, 0") (HitInfo_TraceEnd :type Vector :default "0, 0, 0") (InstigatedBy :type Fort%20Pawn%20Object%20Reference) (DamageCauser :type Actor%20Object%20Reference) (EffectContext :type Gameplay%20Effect%20Context%20Handle%20Structure))
  ;[1 1 1] Damage Health
  (Utilities|IsValid DamageCauser
    (:Is%20Valid
      (GameplayCue|ExecuteGameplayCue%28Burst%29Local 0 (Variables|Default|GetImpactCue) (Ability|GameplayCue|MakeGameplayCueParameters :NormalizedMagnitude 0.0 :RawMagnitude 0.0 :Location HitInfo_Location :Normal (Math|Vector|Normalize (- (Transformation|GetActorLocation DamageCauser) HitInfo_Location)) :GameplayEffectLevel 1 :AbilityLevel 1 :bReplicateLocationWhenUsingMinimalRepProxy false)))
    (:Is%20Not%20Valid
      (GameplayCue|ExecuteGameplayCue%28Burst%29Local 0 (Variables|Default|GetImpactCue) (Ability|GameplayCue|MakeGameplayCueParameters :NormalizedMagnitude 0.0 :RawMagnitude 0.0 :Location HitInfo_Location :Normal HitInfo_ImpactNormal :GameplayEffectLevel 1 :AbilityLevel 1 :bReplicateLocationWhenUsingMinimalRepProxy false)))))

(event EventOnDamaged
  (params (Damage :type Float%20%28single-precision%29) (DamageTags :type Gameplay%20Tag%20Container%20Structure%20%28by%20ref%29) (EffectContext :type Gameplay%20Effect%20Context%20Handle%20Structure%20%28by%20ref%29) (EventInstigator :type Controller%20Object%20Reference) (DamageCauser :type Actor%20Object%20Reference))
  ;[1 1 1] onDamage event on server
  (|CallParentFunction Damage DamageTags EffectContext EventInstigator DamageCauser)
  (if (GameplayTags|HasTag DamageTags "(TagName=\"Weapon.Melee.Impact.Pickaxe\")")
    (CallFunction|HitwithPickaxe (Math|Vector|GetForwardVector (Pawn|GetControlRotation EventInstigator :!class /Script/Engine.Controller)))))

(event EventOnStopTick
  ;[1 1 1] Turn off audio on Stop Tick event
  (Audio|Components|Audio|FadeOut (Variables|Default|GetAudioMovement) 0.5 0.0 "Linear" :!class /Script/Engine.AudioComponent))

(event Custom|TowhookShot
  (bind _returntimeline (Variables|OctopusVehicle_V2|GetReturnTimeline))
  ;[1 1 1] Rope
  (Rendering|Material|SetScalarParameterValueonMaterials (Variables|Default|GetSM_Grapple_Rope) "ReturnWiggle" 0.0)
  (CallFunction|SetGrappleVisible true)
  (if (Components|Timeline|IsPlaying _returntimeline)
    (Components|Timeline|Stop _returntimeline)
    (Development|SetHiddeninGame (Variables|Default|GetCableReturnMeshComponent) true false))
  (CallFunction|TowhookLoop)
  (Variables|Default|SetbHasRopeRetracted false)
  (CallFunction|UpdateTowhookinAirAudio true))

(event Custom|HitWithPickaxe
  (params (HitDirection :type Vector))
  ;[1 1 1] onDamage event on server
  (Audio|BroadcastAbilityImpactatLocation 0 self (* (+ HitDirection "0, 0,0.500000") (Variables|Default|GetPickaxeImpulseStrength)) 777.0)
  (Audio|PlaySoundatLocation "/Game/Athena/Sounds/Vehicles/Hamster/SC_Hamster_CollisionImpact_Cue.SC_Hamster_CollisionImpact_Cue" (Transformation|GetActorLocation)))

(event Custom|TowhookLoop
  (bind _cablestart (Variables|Default|GetCableStart))
  (bind _cableend (Variables|Default|GetCableEnd))
  (bind _returnvalue (- _cableend _cablestart))
  (bind _returnvalue_2 (Math|Vector|VectorLength _returnvalue))
  (bind _returnvalue_3 (/ _returnvalue_2 1e+02))
  (bind (_outcablestart _outcableend) (CallFunction|GetTowhookCableEndpoints))
  (bind _sm_grapple_rope (Variables|Default|GetSM_Grapple_Rope))
  ;[1 1 1] Rope
  (if (CallFunction|IsTowhookHolstered)
    (CallFunction|RetractCable)
    (else
      (Variables|Default|SetCableStart _outcablestart)
      (Variables|Default|SetCableEnd _outcableend)
      (Transformation|SetWorldLocationAndRotation (Variables|Default|GetSM_Grapple_Rope) _cablestart (Math|Rotator|MakeRotfromX _returnvalue) false false)
      (Rendering|Material|SetScalarParameterValueonMaterials _sm_grapple_rope "MaxLength" _returnvalue_2)
      (Transformation|SetRelativeScale3D _sm_grapple_rope (Math|Vector|MakeVector _returnvalue_3 1.0 1.0))
      (bind _returnvalue_4 (Utilities|Time|SetTimerbyEvent (EventDispatchers|CreateEvent :!event Custom|TowhookLoop) 0.0001))
      (Variables|Default|SetCableLoopTimer _returnvalue_4)
      (if (CallFunction|IsTowhookAttached)
        (bind _output_get (Variables|Default|SetWiggleAmount (Math|Interpolation|FInterpTo (Variables|Default|GetWiggleAmount) 0.0 (Utilities|Time|GetWorldDeltaSeconds) 2e+01)))
        (Rendering|Material|SetScalarParameterValueonMaterials (Variables|Default|GetSM_Grapple_Rope) "WiggleAmount" _output_get)
        (else
          (bind _output_get_2 (Variables|Default|SetWiggleAmount (/ _returnvalue_3 (Variables|Default|GetOutgoingWiggleFactor))))
          (Rendering|Material|SetScalarParameterValueonMaterials (Variables|Default|GetSM_Grapple_Rope) "WiggleAmount" _output_get_2))))))

(event Custom|SetGrappleVisible
  (params (Visible :type Boolean))
  (bind _sm_grapple_rope (Variables|Default|GetSM_Grapple_Rope))
  ;[1 1 1] Rope
  (if Visible
    (Rendering|SetVisibility _sm_grapple_rope true false)
    (else
      (Rendering|SetVisibility _sm_grapple_rope false false))))

(event OnOutOfFuel(FortVehicleFuel)
  ;[1 1 1] EVENT -- On Out of Fuel
  (Variables|FX|SetCoil%20FX%20Locked true)
  (Components|Activation|Deactivate (Variables|Default|GetCoilDustVisuals))
  (Audio|Parameter|SetFloatParameter (Variables|Default|GetAudioMovement) "OutOfFuel" 0.0 :!class /Script/AudioExtensions.AudioParameterControllerInterface))

(event Custom|TowhookAttached
  (bind _returnvalue (Vehicle|GetAllPassengers))
  (bind _output (Utilities|Array|Get(acopy) _returnvalue))
  ;[1 1 1] Rope
  (Rendering|Material|SetScalarParameterValueonMaterials (Variables|Default|GetSM_Grapple_Rope) "ReturnWiggle" 0.0)
  (CallFunction|UpdateTowhookinAirAudio false)
  (Utilities|IsValid _output
    (:Is%20Valid
      (if (Pawn|IsLocallyViewed _output)
        (Audio|PlaySound2D (Variables|Sounds|GetTowhookDriverAttachedSound))
        (else
          _)))
    (:Is%20Not%20Valid
      _))
  (Audio|PlaySoundatLocation (Variables|Sounds|GetTowhookImpactSound) (CallFunction|GetTowhookCableAttachLocation)))

(event Custom|TowhookMissed
  (bind _cablereturnmeshcomponent (Variables|Default|GetCableReturnMeshComponent))
  ;[1 1 1] Rope
  (Utilities|Time|ClearandInvalidateTimerbyHandle (Variables|Default|GetCableLoopTimer))
  (CallFunction|RetractCable)
  (Audio|SpawnSoundAttached "/Game/Athena/Sounds/Vehicles/Hamster/SC_Hamster_Grappling_YankEnd_Cue.SC_Hamster_Grappling_YankEnd_Cue" (Variables|Vehicle|GetSkeletalMesh))
  (Utilities|FlowControl|Sequence
    (:then_0
      (Utilities|IsValid _cablereturnmeshcomponent
        (:Is%20Valid
          (Development|SetHiddeninGame _cablereturnmeshcomponent false false))
        (:Is%20Not%20Valid
          (bind _returnvalue (AddComponent|NODE_AddStaticMeshComponent-0 :RelativeTransform_Location "0, 0, 0" :RelativeTransform_Rotation "0, 0, 0" :RelativeTransform_Scale "1.500000,1.500000,1.500000"))
          (bind _output_get (Variables|Default|SetCableReturnMeshComponent _returnvalue))
          (Transformation|DetachFromComponent _output_get "KeepRelative" "KeepRelative" "KeepRelative" false))))
    (:then_1
      (Transformation|SetWorldRotation (Variables|Default|GetCableReturnMeshComponent) (Math|Rotator|MakeRotfromX (- (Variables|Default|GetCableEnd) (Variables|Default|GetCableStart))) false false)
      (Variables|Default|SetbTowhookMissed true))))

(event Custom|TowhookDetached
  (bind _returnvalue (Vehicle|GetAllPassengers))
  (bind _output (Utilities|Array|Get(acopy) _returnvalue))
  (bind _sm_grapple_rope (Variables|Default|GetSM_Grapple_Rope))
  ;[1 1 1] Rope
  (Rendering|Material|SetScalarParameterValueonMaterials _sm_grapple_rope "WiggleAmount" 15.0)
  (Rendering|Material|SetScalarParameterValueonMaterials _sm_grapple_rope "ReturnWiggle" 0.0)
  (CallFunction|UpdateTowhookinAirAudio true)
  (bind _returnvalue_2 (Audio|SpawnSoundAttached "/Game/Athena/Sounds/Vehicles/Hamster/SC_Hamster_Grappling_YankEnd_Cue.SC_Hamster_Grappling_YankEnd_Cue" (Variables|Vehicle|GetSkeletalMesh)))
  (CallFunction|DisplayTowhookAudioVisualizer _returnvalue_2)
  (Utilities|IsValid _output
    (:Is%20Valid
      (if (Pawn|IsLocallyControlled _output)
        (Audio|PlaySound2D "/Game/Athena/Sounds/Vehicles/Hamster/SC_Hamster_Wind_Whoosh_1P_Cue.SC_Hamster_Wind_Whoosh_1P_Cue" (Math|Float|MapRangeClamped (+ (select (Class|FortVehicleInterface|IsBoosting) 2e+03 0.0) (Math|Vector|VectorLength (Transformation|GetVelocity))) 1e+03 3e+03 0.0 1.0))))))

(event Custom|RetractCable
  (bind (_returnvalue_location _returnvalue_rotation) (Math|Transform|BreakTransform (Transformation|GetSocketTransform (Variables|Default|GetSM_Grapple_Rope) "end" "RTS_World")))
  (bind _relativescale3d_x (Math|Vector|BreakVector (Variables|/Script/Engine.SceneComponent|GetRelativeScale3D_X (Variables|Default|GetSM_Grapple_Rope))))
  (bind _sm_grapple_rope (Variables|Default|GetSM_Grapple_Rope))
  ;[1 1 1] Rope
  (bind _output_get (Variables|Default|SetRetractScaleStart _relativescale3d_x))
  (Components|Timeline|SetPlayRate (Variables|OctopusVehicle_V2|GetReturnTimeline) (/ 1.0 (/ _output_get 1e+02)))
  (bind (_direction _newtrack_0) (Timeline|ReturnTimeline PlayFromStart)
    (:Update
      (Transformation|SetRelativeScale3D (Variables|Default|GetSM_Grapple_Rope) (Math|Vector|MakeVector (Math|Float|Lerp (Variables|Default|GetRetractScaleStart) 0.001 _newtrack_0) 1.0 1.0))
      (Rendering|Material|SetScalarParameterValueonMaterials _sm_grapple_rope "WiggleAmount" (Math|Float|Lerp 0.0 15.0 _newtrack_0))
      (Rendering|Material|SetScalarParameterValueonMaterials _sm_grapple_rope "ReturnWiggle" (Math|Float|Lerp 0.0 3.0 _newtrack_0))
      (Rendering|Material|SetScalarParameterValueonMaterials _sm_grapple_rope "MaxLength" (* (Variables|Default|GetRetractScaleStart) 1e+02))
      (if (Variables|Default|GetbTowhookMissed)
        (Transformation|SetWorldTransform (Variables|Default|GetCableReturnMeshComponent) (Math|Transform|MakeTransform (+ _returnvalue_location (* (Math|Vector|GetForwardVector _returnvalue_rotation) 2e+01)) _returnvalue_rotation "1.500000,1.500000,1.500000") false false)))
    (:Finished
      (CallFunction|UpdateTowhookinAirAudio false)
      (Variables|Default|SetbHasRopeRetracted true)
      (CallFunction|SetGrappleVisible false)
      (if (Variables|Default|GetbTowhookMissed)
        (Variables|Default|SetbTowhookMissed false)
        (Development|SetHiddeninGame (Variables|Default|GetCableReturnMeshComponent) true false)))))

(event EventOnDeathPlayEffects
  (params (Damage :type Float%20%28single-precision%29) (DamageTags :type Gameplay%20Tag%20Container%20Structure%20%28by%20ref%29) (Momentum :type Vector :default "0, 0, 0") (HitInfo :type Hit%20Result%20Structure%20%28by%20ref%29) (InstigatedBy :type Fort%20Pawn%20Object%20Reference) (DamageCauser :type Actor%20Object%20Reference) (EffectContext :type Gameplay%20Effect%20Context%20Handle%20Structure))
  (bind _returnvalue (Transformation|GetActorLocation))
  (bind (_returnvalue_location _returnvalue_rotation) (Math|Transform|BreakTransform (Transformation|GetSocketTransform (Variables|Vehicle|GetSkeletalMesh) "DriverSeat" "RTS_World")))
  ;[1 0.3615 0.07292] EVENT -> On Death Play Effects
  (bind _returnvalue_2 (Effects|SpawnEmitteratLocation "/Game/Athena/DrivableVehicles/FX/Octopus/P_Octopus_Vehicle_Destroyed.P_Octopus_Vehicle_Destroyed" _returnvalue_location _returnvalue_rotation "1.000000,1.000000,1.000000" true "AutoRelease"))
  (Effects|Components|ParticleSystem|SetVectorParameter _returnvalue_2 "Velocity" (Transformation|GetVelocity))
  (Camera|PlayWorldCameraShake "/Game/Effects/Fort_Effects/BluePrints/CameraShakes/B_Small_Vertical_Jolt_CameraShake.B_Small_Vertical_Jolt_CameraShake_C" _returnvalue 1.5e+02 7e+02)
  (Fortnite|PlayLocalForceFeedbackatLocationMulti "/Game/Weapons/ForceFeedback/Explosions/FF_Generic_Explosion_Near.FF_Generic_Explosion_Near" "/Game/Weapons/ForceFeedback/Explosions/FF_Generic_Explosion_Far.FF_Generic_Explosion_Far" _returnvalue 5e+02 1.2e+03 "")
  (Development|SetHiddeninGame (Variables|Vehicle|GetSkeletalMesh) true false)
  (Audio|PlaySoundatLocation (Variables|Default|GetDestroySound) (Transformation|GetWorldLocation (Variables|Default|GetAudioMovement))))

(event EventOnEnteredWaterBody
  (params (WaterBody :type Fort%20Water%20Body%20Actor%20Object%20Reference))
  (bind _handlewateraudio (Variables|Sounds|GetHandleWaterAudio))
  ;[1 1 1] Event -> On Entered Water Body -- FX for water interaction (Niagra)
  (CallFunction|OnEnterWater WaterBody)
  (if (not (Networking|IsServer))
    (bind _returnvalue (Utilities|Time|SetTimerbyEvent (delegate Custom|CheckWater) 0.2 true))
    (Variables|FX|SetWater%20FX%20Timer _returnvalue)
    (if (Utilities|Time|IsValidTimerHandle _handlewateraudio)
      (Utilities|Time|ClearandInvalidateTimerbyHandle _handlewateraudio))))

(event EventOnHoldExitStarted
  (params (ExitingPawn :type Fort%20Pawn%20Object%20Reference) (ExitDuration :type Float%20%28single-precision%29))
  ;[1 1 1] EVENT -> OnHoldExitStarted/Stopped
  (Ability|SendGameplayEventtoActor ExitingPawn "(TagName=\"Athena.Vehicle.Biplane.OnHoldExitStarted\")" (Utilities|Struct|MakeGameplayEventData :ContextHandle "()" :InstigatorTags "(GameplayTags=)" :TargetTags "(GameplayTags=)" :EventMagnitude ExitDuration :TargetData "()")))

(event EventShowCooldownCue
  (params (Pawn :type Fort%20Player%20Pawn%20Object%20Reference) (Duration :type Float%20%28single-precision%29))
  ;[1 1 1] Cooldow Cue
  (Ability|SendGameplayEventtoActor Pawn "(TagName=\"GameplayEffect.Vehicle.Boost\")" (Utilities|Struct|MakeGameplayEventData :ContextHandle "()" :InstigatorTags "(GameplayTags=)" :TargetTags "(GameplayTags=)" :EventMagnitude Duration :TargetData "()")))

(event Custom|CheckWater
  ;[1 1 1] Event -> On Entered Water Body -- FX for water interaction (Niagra)
  (CallFunction|OnUpdateWater)
  (CallFunction|UpdateWaterAudioParam))

(event EventOnHoldExitStopped
  (params (ExitingPawn :type Fort%20Pawn%20Object%20Reference))
  ;[1 1 1] EVENT -> OnHoldExitStarted/Stopped
  (Ability|SendGameplayEventtoActor ExitingPawn "(TagName=\"Athena.Vehicle.Biplane.OnHoldExitStopped\")"))

(event EventOnExitedWaterBody
  (params (WaterBody :type Fort%20Water%20Body%20Actor%20Object%20Reference))
  ;[1 1 1] Event -> On Entered Water Body -- FX for water interaction (Niagra)
  (Audio|SpawnSoundAttached "/Game/Athena/Sounds/Vehicles/Hamster/SC_Hamster_Movement_Exit_Water_Cue.SC_Hamster_Movement_Exit_Water_Cue" (Variables|Default|GetAudioMovement))
  (CallFunction|OnExitWater)
  (bind _returnvalue (Utilities|Time|SetTimerbyEvent (delegate Custom|UpdateWaterAudio) 0.2 true))
  (Variables|Sounds|SetHandleWaterAudio _returnvalue)
  (Utilities|Time|ClearandInvalidateTimerbyHandle (Variables|FX|GetWater%20FX%20Timer))
  ;[1 1 1] Prevents Baller from constantly splashing as it lightly bounces in water.
  (Timeline|Splash%20Gate%20Timeline PlayFromStart
    (:Finished
      (Variables|FX|SetCan%20Splash true))))

(event Custom|UpdateWaterAudio
  ;[1 1 1] Event -> On Entered Water Body -- FX for water interaction (Niagra)
  (CallFunction|UpdateWaterAudioParam))

(fn AddCooldownCueEvent
  (params (Player%20Pawn :type Fort%20Player%20Pawn%20Object%20Reference))
  (bind _returnvalue (Environment|ApplyGlobalEnvironmentGameplayEffecttoActor Player%20Pawn (LocalVariables|GetGE_BoostCooldown)))
  (Variables|Default|SetCooldownCueEvent _returnvalue))

(fn BoostVFX
  (params (Activate :type Boolean))
  (bind _boostfx (Variables|Default|GetBoostFX))
  (bind _boostfx_2 (Variables|Default|GetBoostFX))
  (if Activate
    (Utilities|IsValid _boostfx_2
      (:Is%20Valid
        (if (not (Components|Activation|IsActive _boostfx_2))
          (Components|Activation|Activate (Variables|Default|GetBoostFX) true)))
      (:Is%20Not%20Valid
        (bind _returnvalue (Niagara|SpawnSystemAttached "/Game/Athena/DrivableVehicles/FX/Octopus/NS_Octopus_Boost.NS_Octopus_Boost" (Variables|Vehicle|GetSkeletalMesh) "DriverSeat" "0, 0, 0" "0, 0, 0" "KeepRelativeOffset" true true "None" false))
        (Variables|Default|SetBoostFX _returnvalue)))
    (return)
    (else
      (Utilities|IsValid _boostfx
        (:Is%20Valid
          (Components|Activation|Deactivate _boostfx)
          (return))))))

(fn CoasterBoostSound
  (if (> (Vehicle|GetTotalSpeedKmh) (+ (Variables|Coaster|GetSpeed%20Coaster%20FOV%20Target) (LocalVariables|GetSpeed%20Max%20Threshold)))
    (bind _coasterboostaudio (Variables|Default|GetCoasterBoostAudio)
      (:then
        (Audio|Components|Audio|FadeOut _coasterboostaudio 0.5 0.0 "Linear" :!class /Script/Engine.AudioComponent))
      (:else
        _))
    (bind _returnvalue (Audio|SpawnSoundAttached "/BallerV2/Sound/DaisyCoaster_Hamster_Boost_Cue.DaisyCoaster_Hamster_Boost_Cue" (Variables|Default|GetAudioMovement)))
    (Variables|Default|SetCoasterBoostAudio _returnvalue)))

(fn DamageDriver
  (bind _damagestates (Variables|Default|GetDamageStates))
  (bind _returnvalue (Utilities|Array|LastIndex _damagestates))
  (bind _returnvalue_2 (Vehicle|GetDriver))
  (bind _returnvalue_3 (Ability|GetAbilitySystemComponent _returnvalue_2))
  (bind _returnvalue_4 (- _returnvalue 1))
  (Utilities|FlowControl|Sequence
    (:then_0
      ;[0.1927 0.1927 0.1927] (TEMP) Copy / Paste Taken from Handle on Health to determine Damage State Index.
      ; Need to roll both into a refactored / pure function for use in this BP.
      ;
      (bind _output_get (LocalVariables|SetHealth%20Ratio (/ (Class|FortVehicleInterface|GetHealth) (Class|FortVehicleInterface|GetMaxHealth))))
      (if (> _output_get (Utilities|Array|Get(aref) (Variables|Default|GetDamageStates) 1))
        (LocalVariables|SetDesired%20Damage%20State%20Index 0)
        (elif (<= (LocalVariables|GetHealth%20Ratio) (Utilities|Array|Get(aref) (Variables|Default|GetDamageStates) _returnvalue_4))
          (LocalVariables|SetDesired%20Damage%20State%20Index _returnvalue_4)
          (else
            (bind _index (Utilities|FlowControl|ForLoopwithBreak 0 (- (Utilities|Array|LastIndex (Variables|Default|GetDamageStates)) 1))
              (:LoopBody
                (if (and (> (Utilities|Array|Get(aref) (Variables|Default|GetDamageStates) _index) (LocalVariables|GetHealth%20Ratio)) (<= (Utilities|Array|Get(aref) (Variables|Default|GetDamageStates) (+ _index 1)) (LocalVariables|GetHealth%20Ratio)))
                  (LocalVariables|SetDesired%20Damage%20State%20Index _index))))))))
    (:then_1
      ;[0.5 0.5 0.5] (TEMP) Code to Check if We're About to Die
      ; Definitely Gets Refactored -- Has to be a better way to do this
      (if (not (<= (Class|FortVehicleInterface|GetHealth) 0))
        ;[0.3594 0.1255 0.08243] If Damage State degraded a tier, damage the player inside the ball.
        (if (!= (LocalVariables|GetDesired%20Damage%20State%20Index) (Variables|Default|GetDamageStateIndex))
          (Utilities|FlowControl|Sequence
            (:then_0
              (if (> (LocalVariables|GetDesired%20Damage%20State%20Index) (Variables|Default|GetDamageStateIndex))
                (Utilities|IsValid _returnvalue_2
                  (:Is%20Valid
                    (Utilities|IsValid _returnvalue_3
                      (:Is%20Valid
                        ;;[0.4635 0.1886 0.1236] Simple GE to Damage Player
                        (GameplayEffects|ApplyGameplayEffectToTarget :self _returnvalue_3 :GameplayEffectClass "/BallerV2/Vehicle/GE_OctoDamage.GE_OctoDamage_C" :Target _returnvalue_3 :Level 1.0)))))))
            (:then_1
              (Variables|Default|SetDamageStateIndex (LocalVariables|GetDesired%20Damage%20State%20Index)))))
        (else
          (Default|DamageonDeath))))))

(fn DamageOnDeath
  (GameplayEffects|ApplyGameplayEffectToTarget :self (Ability|GetAbilitySystemComponent (Vehicle|GetDriver)) :GameplayEffectClass "/BallerV2/Vehicle/GE_OctoDamage.GE_OctoDamage_C" :Target (Ability|GetAbilitySystemComponent (Vehicle|GetDriver)) :Level 1.0))

(fn DisplayTowhookAudioVisualizer
  (params (InTowhookSoundPlayed :type Audio%20Component%20Object%20Reference))
  (bind _sound (Variables|/Script/Engine.AudioComponent|GetSound InTowhookSoundPlayed))
  (Utilities|IsValid InTowhookSoundPlayed
    (:Is%20Valid
      (Utilities|IsValid _sound
        (:Is%20Valid
          (SoundIndicator|ShowSoundIndicator (Variables|Default|GetTowhook%20Sound%20Indicator) (Transformation|GetActorLocation) _sound 1.0 0))))))

(fn GetAffiliationsToShowFor
  (return (Utilities|Array|MakeArray "Hostile")))

(fn HandleOnHealthChanged
  (bind _damagestates (Variables|Default|GetDamageStates))
  (bind _returnvalue (Utilities|Array|LastIndex _damagestates))
  (bind _returnvalue_2 (- _returnvalue 1))
  (Networking|CanExecuteCosmeticEvents
    (:True
      (LocalVariables|SetHealthRatio (/ (Class|FortVehicleInterface|GetHealth) (Class|FortVehicleInterface|GetMaxHealth)))
      (Utilities|FlowControl|Sequence
        (:then_0
          (if (> (LocalVariables|GetHealthRatio) (Utilities|Array|Get(aref) (Variables|Default|GetDamageStates) 1))
            (LocalVariables|SetDesiredDamageStateIndex 0)
            (elif (<= (LocalVariables|GetHealthRatio) (Utilities|Array|Get(aref) (Variables|Default|GetDamageStates) _returnvalue_2))
              (LocalVariables|SetDesiredDamageStateIndex _returnvalue_2)
              (else
                (bind _index (Utilities|FlowControl|ForLoopwithBreak 0 (- (Utilities|Array|LastIndex (Variables|Default|GetDamageStates)) 1))
                  (:LoopBody
                    (if (and (> (Utilities|Array|Get(aref) (Variables|Default|GetDamageStates) _index) (LocalVariables|GetHealthRatio)) (<= (Utilities|Array|Get(aref) (Variables|Default|GetDamageStates) (+ _index 1)) (LocalVariables|GetHealthRatio)))
                      (LocalVariables|SetDesiredDamageStateIndex _index))))))))
        (:then_1
          (if (!= (LocalVariables|GetDesiredDamageStateIndex) (Variables|Default|GetDamageStateIndex))
            (Utilities|FlowControl|Sequence
              (:then_0
                (Rendering|Material|SetScalarParameterValueonMaterials (Variables|Vehicle|GetSkeletalMesh) "Health" (Utilities|Array|Get(aref) (Variables|Default|GetDamageScalarParameterValues) (LocalVariables|GetDesiredDamageStateIndex)))
                (Effects|Components|ParticleSystem|SetFloatParameter (Variables|Default|GetCoilDustVisuals) "Damage" (Math|Float|SelectFloat 1.0 0.0 (>= (LocalVariables|GetDesiredDamageStateIndex) 2)))
                (Effects|Components|ParticleSystem|SetFloatParameter (Variables|Default|GetCoilDustVisuals) "BadlyDamaged" (Math|Float|SelectFloat 1.0 0.0 (>= (LocalVariables|GetDesiredDamageStateIndex) 3))))
              (:then_1
                (if (> (LocalVariables|GetDesiredDamageStateIndex) (Variables|Default|GetDamageStateIndex))
                  (Effects|SpawnEmitterAttached :EmitterTemplate "/Game/Athena/DrivableVehicles/FX/Octopus/P_Octopus_Damage.P_Octopus_Damage" :AttachToComponent (Variables|Vehicle|GetSkeletalMesh) :AttachPointName "DriverSeat" :Scale "1.000000,1.000000,1.000000" :LocationType "SnapToTarget" :bAutoDestroy true :PoolingMethod "None" :bAutoActivate true)
                  (Audio|SpawnSoundAttached (Variables|Sounds|GetSoundOnDamageThreshold) (Variables|Vehicle|GetSkeletalMesh))))
              (:then_2
                (Variables|Default|SetDamageStateIndex (LocalVariables|GetDesiredDamageStateIndex))))))))))

(fn OnEnterWater
  (params (Water%20Body :type Fort%20Water%20Body%20Actor%20Object%20Reference))
  (bind _output_get (Variables|Default|SetWater%20Body Water%20Body))
  (Variables|FX|SetIs%20In%20Water true)
  (CallFunction|OnUpdateWater)
  (bind _outwatersurfaceinfo (WaterBody|GetWaterSurfaceInfoStruct _output_get (Transformation|GetActorLocation self) true false false))
  (if (Variables|FX|GetCan%20Splash)
    (GameplayCue|ExecuteGameplayCue%28Burst%29Local self (Variables|Default|GetGC%20Enter%20Water) (Ability|GameplayCue|MakeGameplayCueParameters :NormalizedMagnitude (Math|Conversions|ToFloat(Integer) (.WaterBodyIdx (Utilities|Struct|BreakWaterSurfaceInfo _outwatersurfaceinfo))) :RawMagnitude (.WaterDepth (Utilities|Struct|BreakWaterSurfaceInfo _outwatersurfaceinfo)) :Location (Utilities|Struct|BreakWaterSurfaceInfo _outwatersurfaceinfo) :Normal (.WaterPlaneNormal (Utilities|Struct|BreakWaterSurfaceInfo _outwatersurfaceinfo)) :GameplayEffectLevel 1 :AbilityLevel 1 :bReplicateLocationWhenUsingMinimalRepProxy false))))

(fn OnExitWater
  (Variables|FX|SetIs%20In%20Water false)
  (CallFunction|OnUpdateWater)
  (Variables|FX|SetCan%20Splash false))

(fn OnUpdateWater
  (bind _returnvalue (Vehicle|GetPawnatSeat 0))
  (bind _returnvalue_2 (WaterBody|GetWaterImmersionDepth _returnvalue))
  (bind _is_in_water (Variables|FX|GetIs%20In%20Water))
  (bind _returnvalue_3 (<= _returnvalue_2 1.0))
  (bind _returnvalue_4 (and _returnvalue_3 _is_in_water))
  (bind _looping_water_fx_component (Variables|Default|GetLooping%20Water%20FX%20Component))
  (bind _looping_water_fx_component_2 (Variables|Default|GetLooping%20Water%20FX%20Component))
  (if (Networking|IsServer)
    _
    (elif (== (Utilities|IsValid (Variables|Default|GetLooping%20Water%20FX%20Component) :!class /Script/Engine.KismetSystemLibrary) _returnvalue_4)
      ;[1 1 1] Update Water FX Parameters
      (bind _outwatersurfaceinfo (WaterBody|GetWaterSurfaceInfoStruct (Variables|Default|GetWater%20Body) (Transformation|GetActorLocation self) true false false))
      (Effects|Components|ParticleSystem|SetVectorParameter _looping_water_fx_component_2 "User.WaterPlaneLocation" (Utilities|Struct|BreakWaterSurfaceInfo _outwatersurfaceinfo))
      (Effects|Components|ParticleSystem|SetVectorParameter _looping_water_fx_component_2 "User.WaterPlaneNormal" (.WaterPlaneNormal (Utilities|Struct|BreakWaterSurfaceInfo _outwatersurfaceinfo)))
      (Effects|Components|ParticleSystem|SetFloatParameter _looping_water_fx_component_2 "User.WaterDepth" (.WaterDepth (Utilities|Struct|BreakWaterSurfaceInfo _outwatersurfaceinfo)))
      (Effects|Components|ParticleSystem|SetFloatParameter _looping_water_fx_component_2 "User.WaterBodyIndex" (Math|Conversions|ToFloat(Integer) (.WaterBodyIdx (Utilities|Struct|BreakWaterSurfaceInfo _outwatersurfaceinfo))))
      (elif _returnvalue_4
        ;[1 1 1] Activate Water
        (bind _returnvalue_5 (Effects|SpawnFXSystemAttached (Variables|FX|GetLooping%20Water%20FX%20Asset) (Variables|Vehicle|GetSkeletalMesh) "DriverSeat"))
        (bind _asniagara_particle_system_component (Utilities|Casting|CastToNiagaraComponent _returnvalue_5))
        (Variables|Default|SetLooping%20Water%20FX%20Component _asniagara_particle_system_component)
        (elif (Utilities|IsValid _looping_water_fx_component :!class /Script/Engine.KismetSystemLibrary)
          ;[1 1 1] Water No Longer Active
          (Transformation|DetachFromComponent _looping_water_fx_component "KeepWorld" "KeepWorld" "KeepWorld" true)
          (Components|Activation|Deactivate _looping_water_fx_component)
          (Variables|Default|SetLooping%20Water%20FX%20Component 0))))))

(fn OnRep_NewVar
  (Audio|PlaySoundatLocation "/Game/Sounds/Fort_Weapons/GrapplingHook/Grappling_Impact_Cue.Grappling_Impact_Cue" (CallFunction|GetTowhookCableAttachLocation)))

(fn OverrideTowhookRange
  (params (Range :type Float%20%28double-precision%29))
  (Utilities|Struct|SetmembersinTowhookParams (Variables|TowhookModel|GetBaseTowhookParams) Range (+ Range 5e+02)))

(fn RefreshDriveParameters
  (bind _maxspeedoverride (Variables|Default|GetMaxSpeedOverride))
  (bind _lowspeedaccelerationoverride (Variables|Default|GetLowSpeedAccelerationOverride))
  (bind _highspeedaccelerationoverride (Variables|Default|GetHighSpeedAccelerationOverride))
  ;[0 0 0] Set Max Speed, Low and High Speed Acceleration from the Base Spherical Drive Params based on the BallerV2GameData Variables if they are all valid
  (if (Math|Boolean|ANDBoolean (ScalableFloat|IsSettingCurveValid _maxspeedoverride) (ScalableFloat|IsSettingCurveValid _lowspeedaccelerationoverride) (ScalableFloat|IsSettingCurveValid _highspeedaccelerationoverride))
    (Utilities|Struct|SetmembersinSphericalDriveParams (Variables|SphericalDriveModel|GetBaseSphericalDriveParams) (ScalableFloat|GetValueatLevelwithContext _maxspeedoverride) (ScalableFloat|GetValueatLevelwithContext _lowspeedaccelerationoverride) (ScalableFloat|GetValueatLevelwithContext _highspeedaccelerationoverride))))

(fn RefreshTowhookRange
  (bind _towhookrangeoverride (Variables|Default|GetTowhookRangeOverride))
  (if (ScalableFloat|IsSettingCurveValid _towhookrangeoverride)
    (Utilities|Struct|SetmembersinTowhookParams (Variables|TowhookModel|GetBaseTowhookParams) (ScalableFloat|GetValueatLevelwithContext _towhookrangeoverride) (* (ScalableFloat|GetValueatLevelwithContext _towhookrangeoverride) 1.1))))

(fn RemoveCooldownCueEvent
  (params (Player%20Pawn :type Fort%20Player%20Pawn%20Object%20Reference))
  (Ability|SendGameplayEventtoActor Player%20Pawn (LocalVariables|GetGE_BoostCooldown) (Utilities|Struct|MakeGameplayEventData :ContextHandle "()" :InstigatorTags "(GameplayTags=)" :TargetTags "(GameplayTags=)" :EventMagnitude 0.0 :TargetData "()"))
  (GameplayEffects|RemoveActiveGameplayEffect (Ability|GetAbilitySystemComponent Player%20Pawn) (Variables|Default|GetCooldownCueEvent) -1))

(fn ShouldUpdateSpeedTarget
  (params (Target :type Float%20%28double-precision%29))
  (return (and (or (< Target (- (Vehicle|GetTotalSpeedKmh) (LocalVariables|GetMin%20Range))) (> Target (+ (Vehicle|GetTotalSpeedKmh) (LocalVariables|GetMax%20Range)))) (> (Vehicle|GetTotalSpeedKmh) 1e+01))))

(fn ShouldShowSoundIndicator
  (params (PlayerController :type Fort%20Player%20Controller%20Object%20Reference))
  (return true))

(fn SpawnSparksVFX
  (params (Hit%20Actor :type Actor%20Object%20Reference) (Actor%20Location :type Vector) (Actor%20Normal :type Vector) (GameplayCue%20Tag :type Gameplay%20Tag%20Structure))
  (if (< (.z Actor%20Normal) 0.85)
    (GameplayCue|ExecuteGameplayCue%28Burst%29Local Hit%20Actor GameplayCue%20Tag (Ability|GameplayCue|MakeGameplayCueParameters :Location Actor%20Location :Normal Actor%20Normal) true)))

(fn ToggleHUDElements
  (params (Actor :type Actor%20Object%20Reference) (ShouldHide :type Boolean))
  (bind _returnvalue (Fortnite|GetFortPlayerControllerfromActor Actor))
  (Utilities|IsValid _returnvalue
    (:Is%20Valid
      (HUD|SetHUDElementVisibilitywithContext _returnvalue (Variables|Default|GetHideHUD_Tag) ShouldHide "HideHudElements" ""))))

(fn UpdateTowhookInAirAudio
  (params (ShouldPlay :type Boolean))
  (if ShouldPlay
    (bind _towhookinairloop (Variables|Default|GetTowhookInAirLoop)
      (:then
        (Audio|Components|Audio|FadeOut _towhookinairloop 0.25 0.0 "Linear" :!class /Script/Engine.AudioComponent))
      (:else
        _))
    (bind _returnvalue (Audio|SpawnSoundAttached "/Game/Athena/Sounds/Vehicles/Hamster/SC_Hamster_Towhook_InAir_Loop_Cue.SC_Hamster_Towhook_InAir_Loop_Cue" (Variables|Default|GetSM_Grapple_Rope)))
    (Variables|Default|SetTowhookInAirLoop _returnvalue)
    (else
      (bind _towhookinairloop_2 (Variables|Default|GetTowhookInAirLoop)
        (:then
          (Audio|Components|Audio|FadeOut _towhookinairloop_2 0.5 0.0 "Linear" :!class /Script/Engine.AudioComponent))
        (:else)))))

(fn UpdateWaterAudioParam
  ;[1 1 1] Check if in water and update water audio based on check. If we leave water we need to get water audio back to 0 and then clear water exit event timer.
  (bind _audiomovement (Variables|Default|GetAudioMovement)
    (:then
      (bind _output_get (Variables|Sounds|SetWaterSpeedAudio (Math|Interpolation|FInterpTo (Variables|Sounds|GetWaterSpeedAudio) (Math|Float|SelectFloat (Vehicle|GetTotalSpeedKmh) 0.0 (Variables|FX|GetIs%20In%20Water)) (Utilities|Time|GetWorldDeltaSeconds) (Math|Float|SelectFloat 1e+01 2e+01 (Variables|FX|GetIs%20In%20Water)))))
      (Audio|Parameter|SetFloatParameter _audiomovement "WaterSpeed" (Math|Float|MapRangeClamped _output_get 0.0 4e+01 0.0 1.0) :!class /Script/AudioExtensions.AudioParameterControllerInterface)
      (Audio|Parameter|SetBooleanParameter _audiomovement "IsInWater" (Variables|FX|GetIs%20In%20Water) :!class /Script/AudioExtensions.AudioParameterControllerInterface)
      (if (and (<= (Variables|Sounds|GetWaterSpeedAudio) 0) (Utilities|Time|IsValidTimerHandle (Variables|Sounds|GetHandleWaterAudio)))
        (Utilities|Time|ClearandInvalidateTimerbyHandle (Variables|Sounds|GetHandleWaterAudio))))
    (:else)))

(fn ConstructionScript
  (CallFunction|RefreshTowhookRange))
